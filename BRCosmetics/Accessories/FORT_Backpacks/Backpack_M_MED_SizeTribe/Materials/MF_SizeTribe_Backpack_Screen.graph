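# Material function that adds an animated green screen display to the Size Tribe backpack's emissive material attribute.
# A time-of-day scaled green glow forms the base, overlaid with panning hard-edged stripes that fade with camera distance.
# A 2x2 flipbook and a positioned frame texture with a time-driven, stepped fill are masked and combined to drive a brighter green.
# Named reroutes share UVs, the screen texture, an input screen mask and a UV-derived gem mask.
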MaterialFunction /BRCosmetics/Accessories/FORT_Backpacks/Backpack_M_MED_SizeTribe/Materials/MF_SizeTribe_Backpack_Screen
graph {nodes [
  n1 = GetMaterialAttributes(None=FunctionInput() {inputType=FunctionInput_MaterialAttributes}) {attributeGetTypes [#1]}
  n2 = Constant3Vector() {constant=<rgba 0.0579 0.3594 0.0525 1>}
  n3 = Constant() {r=0.18}
  n4 = Multiply(
   A=MaterialFunctionCall(Value_Morning=@n3, Value_Day=Constant(), Value_Evening=@n3, Value_Night=Constant() {r=0.5}) {materialFunction=&Scalar_Time_of_Day}.Result
   B=NamedRerouteUsage()
  )
  n5 = Constant() {r=400}
  n6 = Saturate(None=OneMinus(None=MaterialFunctionCall!(Distance=@n5) {materialFunction=/Game/Effects/Fort_Effects/Materials/_Functions/MF_CameraDistanceFade}.Result))
  n7 = Multiply(A=Panner(Coordinate=NamedRerouteUsage()) {speedX=-0.05, speedY=-0.25}) {constB=65}
  n8 = Step(X=Frac(None=ComponentMask(None=@n7) {r=true})) {constY=0.5}
  n9 = LinearInterpolate(A=Add(A=@n1.EmissiveColor, B=Multiply(A=@n2, B=@n4)), B=@n2, Alpha=Multiply(A=Multiply(A=@n6, B=@n8), B=NamedRerouteUsage()))
  n10 = Constant3Vector() {constant=<rgba 0.0075 0.8796 0.0015 1>}
  n11 = NamedRerouteUsage()
  n12 = Time()
  n13 = Add(A=Multiply(A=NamedRerouteUsage(), B=Constant2Vector() {g=5, r=5}) {constB=2.5162}, B=Constant2Vector() {g=-1.5397, r=-2.9003})
  n14 = ComponentMask(None=@n13) {g=true}
  n15 = MaterialFunctionCall(
   "Animation  Phase (0-1)"=Frac(None=Multiply(A=@n12) {constB=1.75})
   Number of Rows=Constant() {r=2}
   Number of Columns=Constant() {r=2}
   Texture=NamedRerouteUsage()
   UVs=AppendVector(A=OneMinus(None=@n14), B=ComponentMask(None=@n13) {r=true})
  ) {materialFunction=&FlipBook}
  n16 = Multiply(A=@n11, B=ComponentMask(None=@n15.Result) {r=true})
  n17 = NamedRerouteUsage()
  n18 = ScalarParameter() {defaultValue=4.2, parameterName=FrameSizeX}
  n19 = ScalarParameter() {defaultValue=2.6, parameterName=FrameSizeY}
  n20 = Add(
   A=Multiply(A=NamedRerouteUsage(), B=AppendVector(A=@n18, B=@n19)) {constB=2.5162}
   B=AppendVector(A=ScalarParameter() {defaultValue=-3.1481, parameterName=FrameAddX}, B=ScalarParameter() {defaultValue=-1.5828, parameterName=FrameAddY})
  )
  n21 = ComponentMask(None=@n20) {g=true}
  n22 = TextureSample!(UVs=AppendVector(A=OneMinus(None=@n21), B=ComponentMask(None=@n20) {r=true}), Tex=NamedRerouteUsage(), Level, Apply View MipBias) {
   mipValueMode=TMVM_MipLevel
   samplerSource=SSM_Clamp_WorldGroupSettings
   samplerType=SAMPLERTYPE_LinearColor
   texture=/BRCosmetics/Accessories/FORT_Backpacks/Backpack_M_MED_SizeTribe/Textures/T_SizeTribe_Backpack_Screen
  }
  n23 = Step(X=@n22.G) {constY=0.01}
  n24 = Step(Y=@n22.G) {constX=0.5, constY=0.5}
  n25 = Multiply(A=@n23, B=@n24)
  n26 = ComponentMask(None=TextureCoordinate() {uTiling=0.5, vTiling=0.5}) {g=true}
  n27 = Constant() {r=0.35}
  n28 = Constant() {r=0.7}
  n29 = Time() {period=2}
  n30 = TextureCoordinate()
  n31 = LinearInterpolate(A=@n29, B=Add(A=@n29) {constB=2}, Alpha=Step(Y=ComponentMask(None=@n30) {r=true}) {constX=0.9})
  n32 = Add(A=Modulo!(A=Multiply(A=@n31) {constB=0.25}, B=Constant() {r=1})) {constB=-0.5}
  n33 = Step(
   Y=OneMinus(None=MaterialFunctionCall(Input=@n26, Input Low=@n27, Input High=@n28) {materialFunction=&RemapValue_0-1}.Result)
   X=OneMinus(None=Abs(None=@n32))
  )
  n34 = LinearInterpolate(A=@n23, B=Saturate(None=Multiply(A=@n25, B=@n33)), Alpha=Saturate(None=@n24))
  n35 = NamedRerouteUsage()
  n36 = Saturate(None=Add(A=Multiply(A=@n16, B=@n17), B=Multiply(A=@n34, B=@n35)))
  FunctionOutput(None=SetMaterialAttributes!(MaterialAttributes=@n1.MaterialAttributes, EmissiveColor=LinearInterpolate(A=@n9, B=@n10, Alpha=@n36)) {attributeSetTypes [#1]})
  FunctionOutput(None=LinearInterpolate(Alpha=NamedRerouteUsage())) {outputName=TR}
  Multiply(A=ComponentMask(None=@n15.Result) {b=true}, B=@n11)
  NamedRerouteDeclaration(None=TextureCoordinate()) {name=UVs, nodeColor=<rgba 1 0 0 1>}
  n37 = TextureObject() {
   samplerType=SAMPLERTYPE_LinearColor
   texture=&T_SizeTribe_Backpack_Screen
  }
  NamedRerouteDeclaration(None=@n37) {name=Texture, nodeColor=<rgba 0.1484 0 1 1>}
  NamedRerouteDeclaration(None=FunctionInput(Preview=TextureSample!(UVs, Tex, Apply View MipBias) {texture=/BRCosmetics/Accessories/FORT_Backpacks/Backpack_M_MED_SizeTribe/Textures/T_SizeTribe_Backpack_FX}.R) {inputName=ScreenMask, inputType=FunctionInput_Scalar}) {name=ScreenMask, nodeColor=<rgba 0 1 0.1328 1>}
  n38 = NamedRerouteUsage()
  n39 = ComponentMask(None=@n38) {r=true}
  n40 = ComponentMask(None=@n38) {g=true}
  n41 = Step(X=OneMinus(None=@n40)) {constY=0.1}
  n42 = Step(X=@n40) {constY=0.7}
  NamedRerouteDeclaration(None=Multiply(A=Step(X=@n39) {constY=0.8}, B=Multiply(A=@n41, B=@n42))) {name=GemMask, nodeColor=<rgba 1 0 0.1875 1>}
 ]}
