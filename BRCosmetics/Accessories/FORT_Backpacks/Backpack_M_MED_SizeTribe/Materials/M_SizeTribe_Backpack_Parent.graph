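# Parent material for the SizeTribe backpack cosmetic, combining the shared 2023 anime character shader with a backpack screen effect.
# The anime shader runs as Pre, Calculations and Post functions, passing colors, normals, gradients and FX data along.
# Its result enters MF_SizeTribe_Backpack_Screen with a screen mask picked from the FX texture by channel parameter.
# That function's TR output drives temporal responsiveness, alongside the pixel animation flag.
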Material /BRCosmetics/Accessories/FORT_Backpacks/Backpack_M_MED_SizeTribe/Materials/M_SizeTribe_Backpack_Parent
graph {
 properties {
  bAutomaticallySetUsageInEditor=false
  bHasPixelAnimation=true
  bUseMaterialAttributes=true
  bUsedWithClothing=true
  bUsedWithMeshDeformer=true
  bUsedWithMorphTargets=true
  bUsedWithNiagaraMeshParticles=true
  bUsedWithSkeletalMesh=true
  bUsedWithStaticLighting=true
  pixelDepthOffsetMode=PDOM_Legacy
  usedShadingModels=Default Lit
 }
 nodes [
  n1 = MaterialFunctionCall!(CustomUV) {materialFunction=/Game/Global/Materials/Character_Parent_2023/Anime_Parent_2023/MaterialFunctions/MFC_AnimeShader_2023_Pre}
  n2 = MaterialFunctionCall!(CustomLightVector, OverrideNormal_WS, STT=@n1.STT, Normal_Map=@n1.STT, DFL=@n1.STT, FX=@n1.STT) {materialFunction=/Game/Global/Materials/Character_Parent_2023/Anime_Parent_2023/MaterialFunctions/MFC_AnimeShader_2023_Calculations}
  TemporalResponsivenessOutput!(None=MaterialFunctionCall!(
    ScreenMask=ChannelMaskParameter(None=@n1.FX) {parameterName=TA_ScreenMask}.R
    In=MaterialFunctionCall!(
     Shading_Gradient=@n2.Shading_Gradient
     Threshold_Gradient=@n2.Shading_Gradient
     Specular_Gradient=@n2.Shading_Gradient
     Sheen_Mask=@n2.Shading_Gradient
     Thickness=@n2.Shading_Gradient
     DFL_Mask=@n2.Shading_Gradient
     FX=@n2.Shading_Gradient
     Metallic_Gradient=@n2.Shading_Gradient
     RimLight_Gradient=@n2.Shading_Gradient
     NormalBuffer=@n2.Shading_Gradient
     --------
     CINE_LightSources=@n2.Shading_Gradient
     CINE_LightColor=@n2.Shading_Gradient
     ---------
     Color_Lit=@n1.Color_Lit
     Color_Shaded=@n1.Color_Lit
     SRM=@n1.Color_Lit
     Emissive=@n1.Color_Lit
     DFL_Color=@n1.Color_Lit
     Specular_Color=@n1.Color_Lit
     RimLight_Color=@n1.Color_Lit
    ) {materialFunction=/Game/Global/Materials/Character_Parent_2023/Anime_Parent_2023/MaterialFunctions/MFC_AnimeShader_2023_Post}.Result
   ) {materialFunction=&/MF_SizeTribe_Backpack_Screen}.TR)
 ]
}
